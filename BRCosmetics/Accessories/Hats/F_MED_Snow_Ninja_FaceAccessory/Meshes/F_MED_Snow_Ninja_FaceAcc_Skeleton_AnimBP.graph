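# Animation Blueprint for the Snow Ninja face accessory skeleton that adds spring-driven secondary motion to its hood bones.
# It copies the attached parent mesh's pose, then chains bone-relative Anim Dynamics on the mid, right, left and top hood bones.
# Their shared alpha scales with 2D pawn speed and drops to zero while targeting.
# An older root-relative left/right pair is unconnected, and the event graph does nothing.
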AnimBlueprint /BRCosmetics/Accessories/Hats/F_MED_Snow_Ninja_FaceAccessory/Meshes/F_MED_Snow_Ninja_FaceAcc_Skeleton_AnimBP
target_skeleton=&/F_MED_Snow_Ninja_FaceAcc_Skeleton
graphs [
 AnimGraph {nodes [
   n1 = Anim Dynamics - Bone: dyn_hood_l(Alpha=1, ExternalForce="0,0,0") {
    angularSpringConstant=100
    bGravityOverrideInSimSpace=true
    gravityScale=0.05
    linearSpringConstant=100
    simulationSpace=RootRelative
    alphaBoolBlend {blendOption=Linear}
    alphaScaleBiasClamp {bInterpResult=true, bMapRange=true, inRange {max=500}}
    boundBone {boneName=dyn_hood_l}
    componentLinearAccScale=<1 1>
    componentLinearVelScale=<1 1>
    physicsBodyDefinitions [{
      sphereCollisionRadius=0
      boundBone {boneName=dyn_hood_l}
      boxExtents=<5 5 5>
      constraintSetup {angularLimitsMax=<0 0 20>, angularLimitsMin=<0 0 -20>, angularTarget=<0 0 0>, linearAxesMax=<0 0 1>, linearAxesMin=<0 0 -2>}
      localJointOffset=<0 0 -1>
     }]
    relativeSpaceBone {boneName=root}
   }
   Anim Dynamics - Bone: dyn_hood_r(ComponentPose=@n1, Alpha=1, ExternalForce="0,0,0") {
    angularSpringConstant=100
    bGravityOverrideInSimSpace=true
    gravityScale=0.05
    linearSpringConstant=100
    simulationSpace=RootRelative
    alphaBoolBlend {blendOption=Linear}
    alphaScaleBiasClamp {bInterpResult=true, bMapRange=true, inRange {max=500}}
    boundBone {boneName=dyn_hood_r}
    componentLinearAccScale=<1 1>
    componentLinearVelScale=<1 1>
    physicsBodyDefinitions [{
      sphereCollisionRadius=0
      boundBone {boneName=dyn_hood_r}
      boxExtents=<5 5 5>
      constraintSetup {angularLimitsMax=<0 0 20>, angularLimitsMin=<0 0 -20>, angularTarget=<0 0 0>, linearAxesMax=<0 0 2>, linearAxesMin=<0 0 -1>}
      localJointOffset=<0 0 1>
     }]
    relativeSpaceBone {boneName=root}
   }
   n2 = MapRangeClamped(Value=Get PawnSpeed2D(), InRangeA=0, InRangeB=200, OutRangeB=SelectFloat(A=0, B=0.85, bPickA=Get bIsTargeting()))
   n3 = Anim Dynamics - Bone: dyn_hood_mid(ComponentPose=Local To Component(LocalPose=Copy Pose From Mesh() {bUseAttachedParent=true}), Alpha=@n2, ExternalForce="0,0,0") {
    angularSpringConstant=100
    bGravityOverrideInSimSpace=true
    linearSpringConstant=60
    simulationSpace=BoneRelative
    alphaBoolBlend {blendOption=Linear}
    boundBone {boneName=dyn_hood_mid}
    componentLinearAccScale=<1 1 1>
    componentLinearVelScale=<1 1 1>
    physicsBodyDefinitions [{
      sphereCollisionRadius=0
      boundBone {boneName=dyn_hood_mid}
      boxExtents=<5 5 5>
      constraintSetup {angularLimitsMax=<0 2.55 20>, angularLimitsMin=<0 -5 -20>, angularTarget=<0 0 0>, linearAxesMax=<6 2 0>, linearAxesMin=<0 -2 0>}
      localJointOffset=<0 -1 0>
     }]
    relativeSpaceBone {boneName=root}
   }
   n4 = Anim Dynamics - Bone: dyn_hood_r(ComponentPose=@n3, Alpha=@n2, ExternalForce="0,0,0") {
    angularSpringConstant=100
    bGravityOverrideInSimSpace=true
    linearSpringConstant=60
    simulationSpace=BoneRelative
    alphaBoolBlend {blendOption=Linear}
    boundBone {boneName=dyn_hood_r}
    componentLinearAccScale=<1 1 1>
    componentLinearVelScale=<1 1 1>
    physicsBodyDefinitions [{
      sphereCollisionRadius=0
      boundBone {boneName=dyn_hood_r}
      boxExtents=<5 5 5>
      constraintSetup {angularLimitsMax=<0 2.55 20>, angularLimitsMin=<0 -5 -20>, angularTarget=<0 0 0>, linearAxesMax=<3 1 0>, linearAxesMin=<0 -2 0>}
      localJointOffset=<0 -1 0>
     }]
    relativeSpaceBone {boneName=root}
   }
   n5 = Anim Dynamics - Bone: dyn_hood_l(ComponentPose=@n4, Alpha=@n2, ExternalForce="0,0,0") {
    angularSpringConstant=100
    bGravityOverrideInSimSpace=true
    linearSpringConstant=60
    simulationSpace=BoneRelative
    alphaBoolBlend {blendOption=Linear}
    boundBone {boneName=dyn_hood_l}
    componentLinearAccScale=<1 1 1>
    componentLinearVelScale=<1 1 1>
    physicsBodyDefinitions [{
      sphereCollisionRadius=0
      boundBone {boneName=dyn_hood_l}
      boxExtents=<5 5 5>
      constraintSetup {angularLimitsMax=<0 2.55 20>, angularLimitsMin=<0 -5 -20>, angularTarget=<0 0 0>, linearAxesMax=<3 1 0>, linearAxesMin=<0 -2 0>}
      localJointOffset=<0 -1 0>
     }]
    relativeSpaceBone {boneName=root}
   }
   n6 = Anim Dynamics - Bone: dyn_hood_top(ComponentPose=@n5, Alpha=@n2, ExternalForce="0,0,0") {
    angularSpringConstant=100
    bGravityOverrideInSimSpace=true
    linearSpringConstant=60
    simulationSpace=BoneRelative
    alphaBoolBlend {blendOption=Linear}
    boundBone {boneName=dyn_hood_top}
    componentLinearAccScale=<1 1 1>
    componentLinearVelScale=<1 1 1>
    physicsBodyDefinitions [{
      sphereCollisionRadius=0
      boundBone {boneName=dyn_hood_top}
      boxExtents=<5 5 5>
      constraintSetup {angularLimitsMax=<0 2.55 20>, angularLimitsMin=<0 -5 -20>, angularTarget=<0 0 0>, linearAxesMax=<3 1 0>, linearAxesMin=<0 -2 0>}
      localJointOffset=<0 -1 0>
     }]
    relativeSpaceBone {boneName=root}
   }
   Output Pose(Result=Component To Local(ComponentPose=@n6)) {layerGroup=None}
  ]}
 EventGraph {nodes [
   TryGetPawnOwner()
   Event BlueprintUpdateAnimation()
  ]}
]
variables []
